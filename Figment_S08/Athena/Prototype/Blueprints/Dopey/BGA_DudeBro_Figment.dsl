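; Figment DudeBro building game actor, apparently a vent, that can have its disabled state set by DirtShell messages.
; On BeginPlay, the authority listens for the stateful DirtShellMessageStruct event on the global GameplayEventRouter.
; It compares the message Id, case-insensitively, with its own Id or the broadcast key "AllVents".
; On a match it calls SetDisabled with data from the message.
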(event EventBeginPlay
  (|CallParentFunction)
  (Networking|SwitchHasAuthority
    (:Authority
      (bind (_asynctaskproxy _payload) (GameplayEventRouter|StartListeningToStatefulEvent (GameplayEventRouter|GetGlobalEventRouter) "/Figment_S08/Athena/Prototype/Blueprints/Dopey/DirtShellMessageStruct.DirtShellMessageStruct")
        (:OnEventReceived
          _)
        (:OnSavedEventStateExists
          _))
      (if (or (Utilities|String|Equal,CaseInsensitive(String) (.Id_5_CBDA2F21436108043A957C946C5AFD39 (Utilities|Struct|BreakDirtShellMessageStruct _payload)) (Variables|Default|GetId)) (Utilities|String|Equal,CaseInsensitive(String) (.Id_5_CBDA2F21436108043A957C946C5AFD39 (Utilities|Struct|BreakDirtShellMessageStruct _payload)) "AllVents"))
        (CallFunction|SetDisabled (Utilities|Struct|BreakDirtShellMessageStruct _payload))))))

(fn ConstructionScript
  (|CallParentFunction))
