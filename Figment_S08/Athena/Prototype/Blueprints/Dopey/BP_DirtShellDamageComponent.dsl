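; Actor component that relays damage on indexed props, apparently pieces of a dirt shell, through its OnDamageAdded event dispatcher.
; ServerCountDamage broadcasts a given damage amount and dirt index.
; Separate DamageProp0 to DamageProp26 events each hard-code their own index into the same broadcast, giving one entry point per prop.
; The component keeps no damage tally in this graph, and BeginPlay and Tick are empty.
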(event Custom|ServerCountDamage
  (params (DamageApplied :type Integer) (DirtIndex :type Byte))
  (Default|CallOnDamageAdded DamageApplied DirtIndex))

(event EventBeginPlay)

(event EventTick
  (params (DeltaSeconds :type Float%20%28single-precision%29)))

(event Custom|DamageProp0
  (params (DamageApplied :type Integer))
  (Default|CallOnDamageAdded DamageApplied))

(event Custom|DamageProp1
  (params (DamageApplied :type Integer))
  (Default|CallOnDamageAdded DamageApplied 1))

(event Custom|DamageProp2
  (params (DamageApplied :type Integer))
  (Default|CallOnDamageAdded DamageApplied 2))

(event Custom|DamageProp3
  (params (DamageApplied :type Integer))
  (Default|CallOnDamageAdded DamageApplied 3))

(event Custom|DamageProp4
  (params (DamageApplied :type Integer))
  (Default|CallOnDamageAdded DamageApplied 4))

(event Custom|DamageProp5
  (params (DamageApplied :type Integer))
  (Default|CallOnDamageAdded DamageApplied 5))

(event Custom|DamageProp6
  (params (DamageApplied :type Integer))
  (Default|CallOnDamageAdded DamageApplied 6))

(event Custom|DamageProp7
  (params (DamageApplied :type Integer))
  (Default|CallOnDamageAdded DamageApplied 7))

(event Custom|DamageProp8
  (params (DamageApplied :type Integer))
  (Default|CallOnDamageAdded DamageApplied 8))

(event Custom|DamageProp9
  (params (DamageApplied :type Integer))
  (Default|CallOnDamageAdded DamageApplied 9))

(event Custom|DamageProp10
  (params (DamageApplied :type Integer))
  (Default|CallOnDamageAdded DamageApplied 10))

(event Custom|DamageProp11
  (params (DamageApplied :type Integer))
  (Default|CallOnDamageAdded DamageApplied 11))

(event Custom|DamageProp12
  (params (DamageApplied :type Integer))
  (Default|CallOnDamageAdded DamageApplied 12))

(event Custom|DamageProp13
  (params (DamageApplied :type Integer))
  (Default|CallOnDamageAdded DamageApplied 13))

(event Custom|DamageProp14
  (params (DamageApplied :type Integer))
  (Default|CallOnDamageAdded DamageApplied 14))

(event Custom|DamageProp15
  (params (DamageApplied :type Integer))
  (Default|CallOnDamageAdded DamageApplied 15))

(event Custom|DamageProp16
  (params (DamageApplied :type Integer))
  (Default|CallOnDamageAdded DamageApplied 16))

(event Custom|DamageProp17
  (params (DamageApplied :type Integer))
  (Default|CallOnDamageAdded DamageApplied 17))

(event Custom|DamageProp18
  (params (DamageApplied :type Integer))
  (Default|CallOnDamageAdded DamageApplied 18))

(event Custom|DamageProp19
  (params (DamageApplied :type Integer))
  (Default|CallOnDamageAdded DamageApplied 19))

(event Custom|DamageProp20
  (params (DamageApplied :type Integer))
  (Default|CallOnDamageAdded DamageApplied 20))

(event Custom|DamageProp21
  (params (DamageApplied :type Integer))
  (Default|CallOnDamageAdded DamageApplied 21))

(event Custom|DamageProp22
  (params (DamageApplied :type Integer))
  (Default|CallOnDamageAdded DamageApplied 22))

(event Custom|DamageProp23
  (params (DamageApplied :type Integer))
  (Default|CallOnDamageAdded DamageApplied 23))

(event Custom|DamageProp24
  (params (DamageApplied :type Integer))
  (Default|CallOnDamageAdded DamageApplied 24))

(event Custom|DamageProp25
  (params (DamageApplied :type Integer))
  (Default|CallOnDamageAdded DamageApplied 25))

(event Custom|DamageProp26
  (params (DamageApplied :type Integer))
  (Default|CallOnDamageAdded DamageApplied 26))
