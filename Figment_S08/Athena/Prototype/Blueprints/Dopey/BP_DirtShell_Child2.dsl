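; Digsite dirt shell actor that raises a fill once all its dirt mounds are destroyed.
; The server counts prop deaths, then animates the fill live or snaps it at BeginPlay, clearing overlapped chests and optionally building pieces.
; A replicated flag fires client quake cue and audio.
; Scalable toggles gate spawning of recorded anomalies and an EQS-based safe teleport for players in the box, destroying their vehicles.
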(event EventBeginPlay
  (|CallParentFunction)
  (Networking|SwitchHasAuthority
    (:Authority
      (Utilities|FlowControl|Sequence
        (:then_0
          ;[1 1 1] Check to see if the dirt mounds blocking the vent have already been destroyed
          (for _array_element (Variables|Default|GetDirtShellProps)
            (Utilities|IsValid _array_element
              (:Is%20Valid
                (EventDispatchers|BindEventtoOnDied _array_element (EventDispatchers|CreateEvent :!event Custom|DeathEvent)))
              (:Is%20Not%20Valid
                (CallFunction|PropDied)))))
        (:then_1
          (if (== (Variables|Default|GetNumDead) (Variables|Default|GetMaxNumDirt))
            (CallFunction|EventTrigger)))
        (:then_2
          _)
        (:then_3
          (if (ScalableFloat|GetValueasBool (Variables|Default|GetEnableDigsiteAnomalies))
            (Default|SpawnAnomalies)))))
    (:Remote
      _))
  (Collision|SetCollisionEnabled (Variables|Default|GetBox) "NoCollision"))

(event Collision|EventActorBeginOverlap
  (params (OtherActor :type Actor%20Object%20Reference))
  (|CallParentFunction OtherActor) ; This node is disabled and will not be called.\nDrag off pins to build functionality.
)

(event EventTick
  (params (DeltaSeconds :type Float%20%28single-precision%29))
  (|CallParentFunction DeltaSeconds) ; This node is disabled and will not be called.\nDrag off pins to build functionality.
)

(event Custom|PropDied
  (Math|Integer|IncrementInt (Variables|Default|GetNumDead)
    (:%20%20
      (if (== (Variables|Default|GetNumDead) (Variables|Default|GetMaxNumDirt))
        (if (Variables|Default|GetFirstSetupComplete)
          (Utilities|FlowControl|Sequence
            (:then_0
              (CallFunction|EventTrigger))
            (:then_1
              (Variables|Default|SetTriggeredInGame true)
              (Utilities|FlowControl|Delay 4.0)
              (Variables|Default|SetTriggeredInGame false)))
          (else
            (CallFunction|EventTrigger)))))))

(event Custom|DeathEvent
  (params (DamagedActor :type Actor%20Object%20Reference) (Damage :type Float%20%28single-precision%29) (InstigatedBy :type Controller%20Object%20Reference) (DamageCauser :type Actor%20Object%20Reference) (HitLocation :type Vector) (FHitComponent :type Primitive%20Component%20Object%20Reference) (BoneName :type Name) (Momentum :type Vector))
  (CallFunction|PropDied))

(event Custom|TestEventTrigger
  (bind (_origin _boxextent) (Collision|GetActorBounds (Variables|Default|GetFill) true false))
  (bind (_boxextent_x _boxextent_y) (Math|Vector|BreakVector _boxextent))
  (bind _fill (Variables|Default|GetFill))
  (CallFunction|TriggerQuake)
  (Networking|FlushNetDormancy)
  (Networking|SwitchHasAuthority
    (:Authority
      (Variables|Default|SetFillOriginLocation (Transformation|GetActorLocation (Variables|Default|GetFill)))
      (Rendering|SetActorHiddenInGame (Variables|Default|GetFill) false)
      (if (Variables|Default|GetFirstSetupComplete)
        (Development|PrintString "this was triggered live" true true "(R=0.000000,G=0.660000,B=1.000000,A=1.000000)" 2e+01)
        (bind (_direction _raise) (Timeline|Raise PlayFromStart)
          (:Update
            (Transformation|SetActorLocation _fill (+ (Variables|Default|GetFillOriginLocation) (Math|Vector|MakeVector 0.0 0.0 (* _raise (Variables|Default|GetOffsetFill)))) false false)
            (if (ScalableFloat|GetValueasBool (Variables|Default|GetEnableBuildingPropsDestruction))
              (bind _outactors (Collision|BoxOverlapActors (Transformation|GetActorLocation _fill) (Math|Vector|MakeVector _boxextent_x _boxextent_y 1e+02) (Utilities|Array|MakeArray "ObjectTypeQuery1") "/Script/FortniteGame.BuildingActor" :!class /Script/Engine.KismetSystemLibrary))
              (for _array_element _outactors
                (Actor|DestroyActor _array_element))))
          (:Finished
            (bind _outhits (Collision|MultiBoxTraceForObjects :Start (- (Transformation|GetActorLocation (Variables|Default|GetFill)) "0, 0,1000.000000") :End (- (Transformation|GetActorLocation (Variables|Default|GetFill)) "0, 0,1000.000000") :HalfSize "2000.000000,2000.000000,1000.000000" :Orientation "0, 0, 0" :ObjectTypes (Utilities|Array|MakeArray "ObjectTypeQuery1") :bTraceComplex false :DrawDebugType "ForDuration" :bIgnoreSelf true :TraceColor "(R=1.000000,G=0.000000,B=0.000000,A=1.000000)" :TraceHitColor "(R=0.000000,G=1.000000,B=0.000000,A=1.000000)" :DrawTime 6e+01))
            (for _array_element_2 _outhits
              (bind _astiered_chest_athena (Utilities|Casting|CastToTiered_Chest_Athena (.HitActor (Collision|BreakHitResult _array_element_2))))
              (Development|PrintString (Utilities|String|Append (Utilities|GetDisplayName (.HitActor (Collision|BreakHitResult _array_element_2))) " is an object") true true "(R=0.000000,G=0.660000,B=1.000000,A=1.000000)" 6e+01)
              (Actor|DestroyActor _astiered_chest_athena))
            (if (ScalableFloat|GetValueasBool (Variables|Default|GetEnablePlayersSafeTeleportation))
              (Collision|SetCollisionEnabled (Variables|Default|GetBox) "QueryOnly"))))
        (else
          (Development|PrintString "this was triggered at begin play" true true "(R=0.000000,G=0.660000,B=1.000000,A=1.000000)" 2e+01)
          (Transformation|SetActorLocation (Variables|Default|GetFill) (+ (Variables|Default|GetFillOriginLocation) (Math|Vector|MakeVector 0.0 0.0 (Variables|Default|GetOffsetFill))) false false))))))

(event Custom|EventTrigger
  (bind (_origin _boxextent) (Collision|GetActorBounds (Variables|Default|GetFill) true false))
  (bind (_boxextent_x _boxextent_y) (Math|Vector|BreakVector _boxextent))
  (bind _fill (Variables|Default|GetFill))
  (Networking|SwitchHasAuthority
    (:Authority
      (Variables|Default|SetFillOriginLocation (Transformation|GetActorLocation (Variables|Default|GetFill)))
      (Rendering|SetActorHiddenInGame (Variables|Default|GetFill) false)
      (if (Variables|Default|GetFirstSetupComplete)
        (Development|PrintString "this was triggered live" true true "(R=0.000000,G=0.660000,B=1.000000,A=1.000000)" 2e+01)
        (Timeline|Raise PlayFromStart)
        (else
          (Development|PrintString "this was triggered at begin play" true true "(R=0.000000,G=0.660000,B=1.000000,A=1.000000)" 2e+01)
          (Transformation|SetActorLocation (Variables|Default|GetFill) (+ (Variables|Default|GetFillOriginLocation) (Math|Vector|MakeVector 0.0 0.0 (Variables|Default|GetOffsetFill))) false false)
          (bind _outhits (Collision|MultiBoxTraceForObjects :Start (- (Transformation|GetActorLocation (Variables|Default|GetFill)) "0, 0,1000.000000") :End (- (Transformation|GetActorLocation (Variables|Default|GetFill)) "0, 0,1000.000000") :HalfSize "2000.000000,2000.000000,1000.000000" :Orientation "0, 0, 0" :ObjectTypes (Utilities|Array|MakeArray "ObjectTypeQuery1") :bTraceComplex false :DrawDebugType "ForDuration" :bIgnoreSelf true :TraceColor "(R=1.000000,G=0.000000,B=0.000000,A=1.000000)" :TraceHitColor "(R=0.000000,G=1.000000,B=0.000000,A=1.000000)" :DrawTime 6e+01))
          (for _array_element _outhits
            (bind _astiered_chest_athena (Utilities|Casting|CastToTiered_Chest_Athena (.HitActor (Collision|BreakHitResult _array_element))))
            (Development|PrintString (Utilities|String|Append (Utilities|GetDisplayName (.HitActor (Collision|BreakHitResult _array_element))) " is an object") true true "(R=0.000000,G=0.660000,B=1.000000,A=1.000000)" 6e+01)
            (Actor|DestroyActor _astiered_chest_athena))
          (if (ScalableFloat|GetValueasBool (Variables|Default|GetEnablePlayersSafeTeleportation))
            (Collision|SetCollisionEnabled (Variables|Default|GetBox) "QueryOnly")))))))

(event Custom|TriggerQuake
  (Networking|SwitchHasAuthority
    (:Remote
      (GameplayCue|ExecuteGameplayCue%28Burst%29Local (Pawn|GetControlledPawn (Utilities|Array|Get(acopy) (Fortnite|GetLocalFortPlayerControllers true true))) (Variables|Default|GetCueTag_EarthquakeTremors) (Ability|GameplayCue|MakeGameplayCueParameters :NormalizedMagnitude 0.0 :RawMagnitude 0.0 :EffectContext "()" :AggregatedSourceTags "(GameplayTags=)" :AggregatedTargetTags "(GameplayTags=)" :Location "0,0,0" :Normal "0,0,0" :EffectCauser self :GameplayEffectLevel 1 :AbilityLevel 1 :bReplicateLocationWhenUsingMinimalRepProxy false))
      (Audio|Components|Audio|Play (Variables|/Figment_S08/Athena/Prototype/Blueprints/Dopey/BP_Dopey_Fill_B.BP_Dopey_Fill_B_C|GetAudio (Variables|Default|GetFill)) 0.0))))

(event OnComponentBeginOverlap(Box)
  (params (OverlappedComponent :type Primitive%20Component%20Object%20Reference) (OtherActor :type Actor%20Object%20Reference) (OtherComp :type Primitive%20Component%20Object%20Reference) (OtherBodyIndex :type Integer) (bFromSweep :type Boolean) (SweepResult :type Hit%20Result%20Structure%20%28by%20ref%29))
  (Networking|SwitchHasAuthority
    (:Authority
      (if (ScalableFloat|GetValueasBool (Variables|Default|GetEnablePlayersSafeTeleportation))
        (bind _asfort_player_pawn (Utilities|Casting|CastToFortPlayerPawn OtherActor)
          (:then
            (Utilities|Array|Add (Variables|Default|GetPawnsToSafeTeleport) _asfort_player_pawn))
          (:CastFailed
            ; In case we detect a vehicle, teleport its passengers and simply destroy it
            (bind _asfort_athena_vehicle (Utilities|Casting|CastToFortAthenaVehicle OtherActor))
            (Variables|Default|SetPawnsToSafeTeleport (Vehicle|GetAllPassengers _asfort_athena_vehicle))
            (Vehicle|ForceDestroyVehicle _asfort_athena_vehicle true)))
        (bind _returnvalue (AI|EQS|RunEQSQuery "/Figment_S08/Athena/Prototype/Blueprints/Dopey/EQS_DirtShellBeta.EQS_DirtShellBeta" self "AllMatching"))
        (AI|EQS|BindEventtoOnQueryFinished _returnvalue (delegate Custom|OnPlayerDetectionQueryFinished))))))

(event Custom|OnPlayerDetectionQueryFinished
  (params (QueryInstance :type EQS%20Query%20Instance%20Object%20Reference) (QueryStatus :type EEnvQueryStatus%20Enum :default "Processing"))
  (bind (_returnvalue_x _returnvalue_y _returnvalue_z) (Math|Vector|BreakVector (Transformation|GetActorLocation)))
  (if (Utilities|Enum|Equal(Enum) QueryStatus "Success")
    (bind (_resultlocations _returnvalue) (AI|EQS|GetQueryResultsasLocations QueryInstance))
    (for (_array_element _array_index) (Variables|Default|GetPawnsToSafeTeleport)
      (Class|FortPawn|TeleportTo _array_element (Math|Vector|SelectVector (Utilities|Array|Get(acopy) _resultlocations _array_index) (Math|Vector|MakeVector _returnvalue_x _returnvalue_y (+ _returnvalue_z 1e+03)) _returnvalue) (Transformation|GetActorRotation (stale OnComponentBeginOverlap(Box) _asfort_player_pawn)) false false))
    (Utilities|Array|Clear (Variables|Default|GetPawnsToSafeTeleport))))

(fn GrabAnomalyObjectsRefs
  (bind _outactors (Collision|BoxOverlapActors (Transformation|GetActorLocation) "2000.000000,2000.000000,1000.000000" (Utilities|Array|MakeArray "ObjectTypeQuery1") "/Script/Engine.Actor" :!class /Script/Engine.KismetSystemLibrary))
  (for _array_element _outactors
    (if (Utilities|String|Contains (Utilities|GetDisplayName _array_element) "Anomaly")
      (Utilities|Casting|CastToTiered_Chest_6_Figment _array_element
        (:then
          (Utilities|Array|AddUnique (Variables|Default|GetAnomalyChestsTransforms) (Transformation|GetActorTransform _array_element)))
        (:CastFailed
          (bind _asbuilding_actor (Utilities|Casting|CastToBuildingActorClass (Utilities|GetClass _array_element))
            (:then
              (Utilities|Array|Add (Variables|Default|GetPropsAnomalies) (Utilities|Struct|MakeDigsitesAnomaliesStruct _asbuilding_actor (Transformation|GetActorTransform _array_element))))
            (:CastFailed
              (bind _asbuilding_static_mesh__static_geo (Utilities|Casting|CastToBuildingStaticGeoClass (Utilities|GetClass _array_element)))
              (Utilities|Array|Add (Variables|Default|GetStaticGeoAnomalies) (Utilities|Struct|MakeDigsitesAnomaliesStruct _asbuilding_static_mesh__static_geo (Transformation|GetActorTransform _array_element))))))))))

(fn OnRep_TriggeredInGame
  (if (Variables|Default|GetTriggeredInGame)
    (CallFunction|TriggerQuake)))

(fn SpawnAnomalies
  (for _array_element (Variables|Default|GetAnomalyChestsTransforms)
    (Game|SpawnActorfromClass "/FigmentCore/Building/ActorBlueprints/Containers/Tiered_Chest_6_Figment.Tiered_Chest_6_Figment_C" _array_element "Undefined" "MultiplyWithRoot" 0 0 0 0))
  (for _array_element_2 (Variables|Default|GetPropsAnomalies)
    (bind (_array_element_actorclass_2_bd6e59cf45f06109a02a3e83749df807 _array_element_actortransform_5_99b0f90c4bb4e28fcc8683b8d4f56ff7) (Utilities|Struct|BreakDigsitesAnomaliesStruct _array_element_2))
    (Game|SpawnActorfromClass _array_element_actorclass_2_bd6e59cf45f06109a02a3e83749df807 _array_element_actortransform_5_99b0f90c4bb4e28fcc8683b8d4f56ff7 "Undefined" "MultiplyWithRoot" 0 0))
  (for _array_element_3 (Variables|Default|GetStaticGeoAnomalies)
    (bind (_array_element_actorclass_2_bd6e59cf45f06109a02a3e83749df807_2 _array_element_actortransform_5_99b0f90c4bb4e28fcc8683b8d4f56ff7_2) (Utilities|Struct|BreakDigsitesAnomaliesStruct _array_element_3))
    (Game|SpawnActorfromClass _array_element_actorclass_2_bd6e59cf45f06109a02a3e83749df807_2 _array_element_actortransform_5_99b0f90c4bb4e28fcc8683b8d4f56ff7_2 "Undefined" "MultiplyWithRoot" 0 0)))

(fn ConstructionScript
  (|CallParentFunction))
