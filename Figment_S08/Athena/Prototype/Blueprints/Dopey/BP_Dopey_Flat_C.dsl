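; Prototype "Dopey Flat" damageable actor that plays hit sounds and always reports that it should not die.
; On damage effects, it prints a debug string, then plays the Dopey Imp and Res sound cues at the hit's impact point.
; ShouldDie always returns false, and the BeginPlay, overlap, Tick and construction script entry points are empty.
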(event EventBeginPlay)

(event Collision|EventActorBeginOverlap
  (params (OtherActor :type Actor%20Object%20Reference)))

(event EventTick
  (params (DeltaSeconds :type Float%20%28single-precision%29)))

(event EventOnDamagePlayEffects
  (params (Damage :type Float%20%28single-precision%29) (DamageTags :type Gameplay%20Tag%20Container%20Structure%20%28by%20ref%29) (Momentum :type Vector :default "0, 0, 0") (HitInfo :type Hit%20Result%20Structure%20%28by%20ref%29) (InstigatedBy :type Fort%20Pawn%20Object%20Reference) (DamageCauser :type Actor%20Object%20Reference) (EffectContext :type Gameplay%20Effect%20Context%20Handle%20Structure))
  (bind (_bblockinghit _binitialoverlap _time _distance _location _impactpoint) (Collision|BreakHitResult HitInfo))
  (Development|PrintString "we hit those!")
  (Utilities|FlowControl|Sequence
    (:then_0
      (Audio|PlaySoundatLocation "/Figment_S08/Athena/Sounds/Events/Dopey/Dopey_Imp_Cue.Dopey_Imp_Cue" _impactpoint))
    (:then_1
      (Audio|PlaySoundatLocation "/Figment_S08/Athena/Sounds/Events/Dopey/Dopey_Res_Cue.Dopey_Res_Cue" _impactpoint))))

(fn ShouldDie
  (params (Damage :type Float%20%28single-precision%29) (EventInstigator :type Controller%20Object%20Reference) (DamageCauser :type Actor%20Object%20Reference) (bIsTest :type Boolean) (EffectContext :type Gameplay%20Effect%20Context%20Handle%20Structure%20%28by%20ref%29))
  (return false))

(fn ConstructionScript)
