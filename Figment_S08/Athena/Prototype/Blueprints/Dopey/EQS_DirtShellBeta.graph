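# Environment query in the Dopey prototype, named for a beta dirt shell, that picks a nearby location from a simple grid.
# Grid points are projected vertically onto geometry by tracing a custom game channel, then raised off the hit surface.
# A capsule overlap test against complex WorldStatic collision filters and scores them.
# A 2D distance test scores points inverse-linearly from the querier, favoring closer ones.
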EnvQuery /Figment_S08/Athena/Prototype/Blueprints/Dopey/EQS_DirtShellBeta
options [{
  generator {
   name=EnvQueryGenerator_SimpleGrid_1
   class=EnvQueryGenerator_SimpleGrid
   properties {
    gridSize=666.8881
    spaceBetween=140
    projectionData {
     bOnlyBlockingHits=false
     extentX=0
     navigationFilter=null
     postProjectionVerticalOffset=100
     projectDown=256
     projectUp=256
     serializedChannel=ECC_GameTraceChannel11
     traceChannel=TraceTypeQuery10
     traceMode=GeometryByChannel
    }
   }
  }
  tests [
   EnvQueryTest_Overlap_1 {
    class=EnvQueryTest_Overlap
    properties {
     testPurpose=FilterAndScore
     overlapData {bOverlapComplex=true, extentX=50, extentY=50, extentZ=100, overlapChannel=ECC_WorldStatic, overlapShape=Capsule, shapeOffset=<0 0 100>}
    }
   }
   EnvQueryTest_Distance_2 {
    class=EnvQueryTest_Distance
    properties {
     distanceTo=&EnvQueryContext_Querier
     filterType=Range
     floatValueMin=0
     scoringEquation=InverseLinear
     scoringFactor=1
     testComment=""
     testMode=Distance2D
     testPurpose=Score
    }
   }
  ]
 }]
